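# Physics asset for the F_MED_Disguise backpack mesh, covering its two dangling string bone chains.
# Each bone has one small capsule body, with six bodies on the left chain and five on the right.
# Constraints link each bone to the one before it within its side, with no link between the chains.
# Swing and twist motions are all Free, so the stored angle limits do not restrict rotation.
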PhysicsAsset /Game/Accessories/FORT_Backpacks/Backpack_F_MED_Disguise/Meshes/F_MED_Disguise_Pack_Physics1
bodies {
 dyn_string_01_LFT [{
   shapeType=Capsule
   length=0.5
   radius=0.5
   shapeName=dyn_string_01_LFT_capsule
   center=<0.7465 -0.022 0.1526>
   rotation=<rot -9.9677 -90 73.575>
  }]
 dyn_string_01_RGT [{
   shapeType=Capsule
   length=0.6269
   radius=0.3931
   shapeName=dyn_string_01_RGT_capsule
   center=<0.811 -0.0142 0.006>
   rotation=<rot -2.0645 -90 83.1351>
  }]
 dyn_string_02_LFT [{
   shapeType=Capsule
   length=2.255
   radius=0.505
   shapeName=dyn_string_02_LFT_capsule
   center=<1.6107 0.0876 0.1191>
   rotation=<rot 0 91.0807 -87.5665>
  }]
 dyn_string_02_RGT [{
   shapeType=Capsule
   length=2.5031
   radius=0.4287
   shapeName=dyn_string_02_RGT_capsule
   center=<2.076 -0.1482 0.0718>
   rotation=<rot 0 86.7251 -87.3911>
  }]
 dyn_string_03_LFT [{
   shapeType=Capsule
   length=2.005
   radius=0.505
   shapeName=dyn_string_03_LFT_capsule
   center=<1.3705 0.054 -0.1446>
   rotation=<rot 0 -89.8332 -86.2115>
  }]
 dyn_string_03_RGT [{
   shapeType=Capsule
   length=2.255
   radius=0.505
   shapeName=dyn_string_03_RGT_capsule
   center=<1.5982 0.0451 0>
   rotation=<rot 0 -89.9999 89.9996>
  }]
 dyn_string_04_LFT [{
   shapeType=Capsule
   length=1.505
   radius=0.505
   shapeName=dyn_string_04_LFT_capsule
   center=<0.5819 0.0549 0.051>
   rotation=<rot 0 96.0187 -84.5221>
  }]
 dyn_string_04_RGT [{
   shapeType=Capsule
   length=2.2336
   radius=0.5
   shapeName=dyn_string_04_RGT_capsule
   center=<0.3625 0.0183 -0.0544>
   rotation=<rot 0 88.0232 -87.9816>
  }]
 dyn_string_05_LFT [{
   shapeType=Capsule
   length=1.4416
   radius=0.6403
   shapeName=dyn_string_05_LFT_capsule
   center=<0.9371 -0.0358 -0.0897>
   rotation=<rot 0 -89.3424 -82.6289>
  }]
 dyn_string_05_RGT [{shapeType=Capsule, length=0.505, radius=0.505, shapeName=dyn_string_05_RGT_capsule, rotation=<rot 0 -89.9999 50.0003>}]
 dyn_string_06_LFT [{
   shapeType=Capsule
   length=0.755
   radius=0.505
   shapeName=dyn_string_06_LFT_capsule
   center=<0.2117 0.0001 -0.1471>
   rotation=<rot 0 -89.9999 90.0003>
  }]
}
constraints [
 {
  bone1Name=dyn_string_02_RGT
  bone2Name=dyn_string_01_RGT
  swing1LimitDegrees=30
  swing1Motion=Free
  swing2LimitDegrees=10
  swing2Motion=Free
  twistLimitDegrees=5
  twistMotion=Free
 }
 {
  bone1Name=dyn_string_03_RGT
  bone2Name=dyn_string_02_RGT
  swing1LimitDegrees=30
  swing1Motion=Free
  swing2LimitDegrees=10
  swing2Motion=Free
  twistLimitDegrees=5
  twistMotion=Free
 }
 {
  bone1Name=dyn_string_04_RGT
  bone2Name=dyn_string_03_RGT
  swing1LimitDegrees=30
  swing1Motion=Free
  swing2LimitDegrees=10
  swing2Motion=Free
  twistLimitDegrees=5
  twistMotion=Free
 }
 {
  bone1Name=dyn_string_05_RGT
  bone2Name=dyn_string_04_RGT
  swing1LimitDegrees=30
  swing1Motion=Free
  swing2LimitDegrees=10
  swing2Motion=Free
  twistLimitDegrees=5
  twistMotion=Free
 }
 {
  bone1Name=dyn_string_02_LFT
  bone2Name=dyn_string_01_LFT
  swing1LimitDegrees=30
  swing1Motion=Free
  swing2LimitDegrees=10
  swing2Motion=Free
  twistLimitDegrees=5
  twistMotion=Free
 }
 {
  bone1Name=dyn_string_03_LFT
  bone2Name=dyn_string_02_LFT
  swing1LimitDegrees=30
  swing1Motion=Free
  swing2LimitDegrees=10
  swing2Motion=Free
  twistLimitDegrees=5
  twistMotion=Free
 }
 {
  bone1Name=dyn_string_04_LFT
  bone2Name=dyn_string_03_LFT
  swing1LimitDegrees=30
  swing1Motion=Free
  swing2LimitDegrees=10
  swing2Motion=Free
  twistLimitDegrees=5
  twistMotion=Free
 }
 {
  bone1Name=dyn_string_05_LFT
  bone2Name=dyn_string_04_LFT
  swing1LimitDegrees=30
  swing1Motion=Free
  swing2LimitDegrees=10
  swing2Motion=Free
  twistLimitDegrees=5
  twistMotion=Free
 }
 {
  bone1Name=dyn_string_06_LFT
  bone2Name=dyn_string_05_LFT
  swing1LimitDegrees=30
  swing1Motion=Free
  swing2LimitDegrees=10
  swing2Motion=Free
  twistLimitDegrees=5
  twistMotion=Free
 }
]
